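; Shooting gallery quest interactable that runs a timed target game while a pawn stands on its pressure plate.
; Gated capsule overlaps store the competitor, start the scoreboard and set bActive, whose OnRep starts flipping or deactivates the boards and ends the game.
; On authority, FlipBoard activates two randomly rolled boards, rerolls and re-arms itself using per-round timers, with rounds capped at five.
; TargetEnd forwards scores to the scoreboard, treating -5 as game over.
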(event EventBeginPlay
  ;[0.5993 0.8137 1] Startup, populate arrays
  (for (_array_element _array_index) (Variables|Default|GetTarget%20Boards)
    (Variables|/Figment_S08/Athena/Items/QuestInteractables/ShootingGallery/Blueprint/BP_ShootingTarget.BP_ShootingTarget_C|SetGamePad self _array_element)
    (Utilities|Array|SetArrayElem (Variables|Default|GetActiveTarget) _array_index 0 true)
    (CallFunction|RollRandoms)))

(event Custom|RollRandoms
  (bind _popint (Variables|Default|GetPopInt))
  ;[0.7904 1 0.2395 0.266] Core Game Loop
  ;;[0.1851 1 0.9403] Get first random board #
  (Networking|SwitchHasAuthority
    (:Authority
      (Utilities|Array|Resize _popint)
      (Utilities|Array|SetArrayElem _popint 0 (Math|Random|RandomIntegerinRange 0 (- (Utilities|Array|Length (Variables|Default|GetTarget%20Boards)) 1)) true)
      (CallFunction|RollSecond))))

(event Custom|FlipBoard
  (bind _popint (Variables|Default|GetPopInt))
  ;[0.7904 1 0.2395 0.266] Core Game Loop
  ;;[1 0.7244 0.2645] Board Activate Loop
  (if (and (Variables|Default|GetbActive) (Variables|Default|GetGameActive))
    (Networking|SwitchHasAuthority
      (:Authority
        (for _array_element _popint
          (Class|BPShootingTarget|ActivateBoard (Utilities|Array|Get(acopy) (Variables|Default|GetTarget%20Boards) _array_element) _array_element (Variables|Default|GetRound) :!class /Figment_S08/Athena/Items/QuestInteractables/ShootingGallery/Blueprint/BP_ShootingTarget.BP_ShootingTarget_C))
        (Utilities|Array|Clear _popint)
        (CallFunction|RollRandoms)
        (Utilities|Time|SetTimerbyFunctionName 0 "FlipBoard" (Utilities|Array|Get(aref) (Variables|Default|GetRoundTimers) (Variables|Default|GetRound)))
        (Variables|Default|SetRound (Math|Integer|SelectInt 5 (+ (Variables|Default|GetRound) 1) (>= (Variables|Default|GetRound) 5)))))))

(event Custom|RollSecond
  ;[0.7904 1 0.2395 0.266] Core Game Loop
  ;;[0.3291 0.398 1] Pick second board
  (Networking|SwitchHasAuthority
    (:Authority
      (Utilities|Array|SetArrayElem (Variables|Default|GetPopInt) 1 (Math|Integer|SelectInt (Math|Integer|%25%28Integer%29 (+ (Math|Random|RandomIntegerinRange 0 (- (Utilities|Array|Length (Variables|Default|GetTarget%20Boards)) 1)) 1) (- (Utilities|Array|Length (Variables|Default|GetTarget%20Boards)) 1)) (Math|Random|RandomIntegerinRange 0 (- (Utilities|Array|Length (Variables|Default|GetTarget%20Boards)) 1)) (== (Math|Random|RandomIntegerinRange 0 (- (Utilities|Array|Length (Variables|Default|GetTarget%20Boards)) 1)) (Utilities|Array|Get(aref) (Variables|Default|GetPopInt)))) true))))

(event OnComponentBeginOverlap(Capsule)
  (params (OverlappedComponent :type Primitive%20Component%20Object%20Reference) (OtherActor :type Actor%20Object%20Reference) (OtherComp :type Primitive%20Component%20Object%20Reference) (OtherBodyIndex :type Integer) (bFromSweep :type Boolean) (SweepResult :type Hit%20Result%20Structure%20%28by%20ref%29))
  ;[0.7904 1 0.2395] Stands on the pedestal to start the game
  (Utilities|FlowControl|Gate false
    (:Exit
      (bind _asfort_pawn (Utilities|Casting|CastToFortPawn OtherActor))
      (bind _output_get (Variables|Default|SetCompetitor _asfort_pawn))
      (CallFunction|StepGate)
      (CallFunction|OpenEnd)
      (Variables|Default|SetRound 0)
      (Variables|Default|SetGameActive true)
      ;;[0 0.9646 1] AUDIO - On Pressure Pad
      (Audio|PlaySoundatLocation "/Figment_S08/Athena/Sounds/Quests/Athena_ShootingGallery_PressurePlate_ON__Cue.Athena_ShootingGallery_PressurePlate_ON__Cue" (Transformation|GetActorLocation))
      (Networking|SwitchHasAuthority
        (:Authority
          (Class|ScoreboardInterface|Startup(Message) (Variables|Default|GetScoreboard) _output_get :!class /Figment_S08/Athena/Items/QuestInteractables/ScoreBoard/ScoreboardInterface.ScoreboardInterface_C)
          (Utilities|FlowControl|Delay 0.2)
          (Variables|Default|SetbActive true))
        (:Remote
          (Utilities|FlowControl|Delay 0.2)
          (Transformation|SetRelativeLocation (Variables|Components|Getpressure_plate_button_01) "0, 0,-6.000000" false false) ; Depress the button
        )))))

(event Custom|ClearGate
  ;[0.7904 1 0.2395] Stands on the pedestal to start the game
  (Utilities|FlowControl|Gate false
    (:Exit
      (bind _asfort_pawn (Utilities|Casting|CastToFortPawn (stale OnComponentBeginOverlap(Capsule) OtherActor)))
      (bind _output_get (Variables|Default|SetCompetitor _asfort_pawn))
      (CallFunction|StepGate)
      (CallFunction|OpenEnd)
      (Variables|Default|SetRound 0)
      (Variables|Default|SetGameActive true)
      ;;[0 0.9646 1] AUDIO - On Pressure Pad
      (Audio|PlaySoundatLocation "/Figment_S08/Athena/Sounds/Quests/Athena_ShootingGallery_PressurePlate_ON__Cue.Athena_ShootingGallery_PressurePlate_ON__Cue" (Transformation|GetActorLocation))
      (Networking|SwitchHasAuthority
        (:Authority
          (Class|ScoreboardInterface|Startup(Message) (Variables|Default|GetScoreboard) _output_get :!class /Figment_S08/Athena/Items/QuestInteractables/ScoreBoard/ScoreboardInterface.ScoreboardInterface_C)
          (Utilities|FlowControl|Delay 0.2)
          (Variables|Default|SetbActive true))
        (:Remote
          (Utilities|FlowControl|Delay 0.2)
          (Transformation|SetRelativeLocation (Variables|Components|Getpressure_plate_button_01) "0, 0,-6.000000" false false) ; Depress the button
        )))))

(event Custom|StepGate
  ;[0.7904 1 0.2395] Stands on the pedestal to start the game
  (Utilities|FlowControl|Gate false
    (:Exit
      (bind _asfort_pawn (Utilities|Casting|CastToFortPawn (stale OnComponentBeginOverlap(Capsule) OtherActor)))
      (bind _output_get (Variables|Default|SetCompetitor _asfort_pawn))
      (CallFunction|StepGate)
      (CallFunction|OpenEnd)
      (Variables|Default|SetRound 0)
      (Variables|Default|SetGameActive true)
      ;;[0 0.9646 1] AUDIO - On Pressure Pad
      (Audio|PlaySoundatLocation "/Figment_S08/Athena/Sounds/Quests/Athena_ShootingGallery_PressurePlate_ON__Cue.Athena_ShootingGallery_PressurePlate_ON__Cue" (Transformation|GetActorLocation))
      (Networking|SwitchHasAuthority
        (:Authority
          (Class|ScoreboardInterface|Startup(Message) (Variables|Default|GetScoreboard) _output_get :!class /Figment_S08/Athena/Items/QuestInteractables/ScoreBoard/ScoreboardInterface.ScoreboardInterface_C)
          (Utilities|FlowControl|Delay 0.2)
          (Variables|Default|SetbActive true))
        (:Remote
          (Utilities|FlowControl|Delay 0.2)
          (Transformation|SetRelativeLocation (Variables|Components|Getpressure_plate_button_01) "0, 0,-6.000000" false false) ; Depress the button
        )))))

(event OnComponentEndOverlap(Capsule)
  (params (OverlappedComponent :type Primitive%20Component%20Object%20Reference) (OtherActor :type Actor%20Object%20Reference) (OtherComp :type Primitive%20Component%20Object%20Reference) (OtherBodyIndex :type Integer))
  ;[0.7904 1 0.2395] Step off the pedestal, game stops
  (bind _asfort_pawn (Utilities|Casting|CastToFortPawn OtherActor))
  (if (Utilities|Equal(Object) _asfort_pawn (Variables|Default|GetCompetitor))
    (CallFunction|ClearGate)
    ;;[0 0.9646 1] AUDIO - Off Pressure Pad
    (Audio|PlaySoundatLocation "/Figment_S08/Athena/Sounds/Quests/Athena_ShootingGallery_PressurePlate_OFF_Cue.Athena_ShootingGallery_PressurePlate_OFF_Cue" (Transformation|GetActorLocation))
    (Networking|SwitchHasAuthority
      (:Authority
        (Variables|Default|SetbActive false))
      (:Remote
        (Transformation|SetRelativeLocation (Variables|Components|Getpressure_plate_button_01) "0, 0, 0" false false)))))

(event EventTargetEnd
  (params (Score :type Integer) (BoardID :type Integer))
  (bind _activetarget (Variables|Default|GetActiveTarget))
  (bind _scoreboard (Variables|Default|GetScoreboard))
  ;[1 1 1] Scoring Behavior
  (if (== Score -5)
    (Variables|Default|SetGameActive false)
    (Networking|SwitchHasAuthority
      (:Authority
        (Utilities|FlowControl|DoOnce false
          (:Completed
            (Class|BPScoreBoardPlace|EndGame _scoreboard :!class /Figment_S08/Athena/Items/QuestInteractables/ScoreBoard/BP_ScoreBoard_Place.BP_ScoreBoard_Place_C)
            (for (_ _array_index) _activetarget
              (Utilities|Array|SetArrayElem _activetarget _array_index))))))
    (else
      (Networking|SwitchHasAuthority
        (:Authority
          (Class|BPScoreBoardPlace|MoveScore _scoreboard Score (Utilities|Array|MakeArray 0) :!class /Figment_S08/Athena/Items/QuestInteractables/ScoreBoard/BP_ScoreBoard_Place.BP_ScoreBoard_Place_C))))))

(event Custom|OpenEnd
  (bind _activetarget (Variables|Default|GetActiveTarget))
  ;[1 1 1] Scoring Behavior
  (Utilities|FlowControl|DoOnce false
    (:Completed
      (Class|BPScoreBoardPlace|EndGame (Variables|Default|GetScoreboard) :!class /Figment_S08/Athena/Items/QuestInteractables/ScoreBoard/BP_ScoreBoard_Place.BP_ScoreBoard_Place_C)
      (for (_ _array_index) _activetarget
        (Utilities|Array|SetArrayElem _activetarget _array_index)))))

(event Collision|EventActorBeginOverlap
  (params (OtherActor :type Actor%20Object%20Reference)))

(fn OnRep_bActive
  (if (Variables|Default|GetbActive)
    (CallFunction|FlipBoard)
    (else
      (CallFunction|TargetEnd -5)
      (for (_array_element _array_index) (Variables|Default|GetTarget%20Boards)
        (Class|BPShootingTarget|DeactivateBoard _array_element :!class /Figment_S08/Athena/Items/QuestInteractables/ShootingGallery/Blueprint/BP_ShootingTarget.BP_ShootingTarget_C)
        (Utilities|Array|SetArrayElem (Variables|Default|GetActiveTarget) _array_index)))))

(fn OnRep_PopInt)

(fn ConstructionScript)
